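; Blueprint for the small half-shield consumable weapon in the Paprika consumables set.
; On the equipped-and-attached event, it gets the weapon mesh's anim instance and stops the synced consume montage instantly if it is still playing.
; This keeps a drinking animation from carrying over when the item is re-equipped.
; The construction script only calls its parent, and no consume or shield logic appears in this graph.
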(event FortWeapon|EventOnEquippedAndAttached
  (bind _weaponmesh (Variables|Visual|GetWeaponMesh))
  (bind _returnvalue (Components|SkeletalMesh|GetAnimInstance _weaponmesh))
  (bind _syncedconsumemontage (Variables|Default|GetSyncedConsumeMontage))
  ;[0.03429 0.3069 0.7899] Ensure we stop drinking montage when reequipped
  (if (Animation|Montage|MontageIsPlaying _returnvalue _syncedconsumemontage)
    (Animation|Montage|MontageStop _returnvalue 0.0 _syncedconsumemontage)))

(fn ConstructionScript
  (|CallParentFunction))
