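; Tossed shield potion projectile that plays a toss sound, destroys itself when it stops moving, and leaves a poof effect and landing sound.
; BeginPlay creates a dynamic material on slot 2 of the shield mesh and sets its liquid fill to zero.
; A replicated fill amount drives UpdateFillAmount, which writes 0.66 of that value to LiquidMaterialInstance, a variable this graph never assigns.
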(event EventBeginPlay
  (Audio|SpawnSoundAttached "/PaprikaConsumables/Assets/Sound/ShieldPot_Tossed_Cue.ShieldPot_Tossed_Cue" (Variables|Default|GetSound))
  (bind _returnvalue (Rendering|Material|CreateDynamicMaterialInstance (Variables|Default|GetSK_Shields) 2 (Rendering|Material|GetMaterial (Variables|Default|GetSK_Shields) 2) "None" :!class /Script/Engine.PrimitiveComponent))
  (Rendering|Material|SetScalarParameterValue _returnvalue "Liquid Fill" 0.0 :!class /Script/Engine.MaterialInstanceDynamic))

(event EventDestroyed
  (Niagara|SpawnSystematLocation "/Game/Animation/Props/Gadgets_Troops/FX/NS_Emote_Troops_Poof.NS_Emote_Troops_Poof" (Transformation|GetActorLocation))
  (Audio|PlaySoundatLocation "/PaprikaConsumables/Assets/Sound/ShieldPot_Tossed_Land_Cue.ShieldPot_Tossed_Land_Cue" (Transformation|GetWorldLocation (Variables|Default|GetSound))))

(event EventOnStop
  (params (Hit_bBlockingHit :type Boolean) (Hit_bInitialOverlap :type Boolean) (Hit_Time :type Float%20%28single-precision%29) (Hit_Distance :type Float%20%28single-precision%29) (Hit_Location :type Vector :default "0, 0, 0") (Hit_ImpactPoint :type Vector :default "0, 0, 0") (Hit_Normal :type Vector :default "0, 0, 0") (Hit_ImpactNormal :type Vector :default "0, 0, 0") (Hit_PhysMat :type Physical%20Material%20Object%20Reference) (Hit_HitActor :type Actor%20Object%20Reference) (Hit_HitComponent :type Primitive%20Component%20Object%20Reference) (Hit_HitBoneName :type Name :default "None") (Hit_BoneName :type Name :default "None") (Hit_HitItem :type Integer) (Hit_ElementIndex :type Integer) (Hit_FaceIndex :type Integer) (Hit_TraceStart :type Vector :default "0, 0, 0") (Hit_TraceEnd :type Vector :default "0, 0, 0"))
  (Actor|DestroyActor))

(fn OnRep_FillAmount
  (CallFunction|UpdateFillAmount (Variables|Default|GetFill%20Amount)))

(fn UpdateFillAmount
  (params (FillPercent :type Float%20%28double-precision%29))
  (Rendering|Material|SetScalarParameterValue (Variables|Default|GetLiquidMaterialInstance) "Liquid Fill" (* FillPercent 0.66) :!class /Script/Engine.MaterialInstanceDynamic)
  (return))

(fn ConstructionScript)
