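; Gameplay ability for the Paprika shield consumable, layering cosmetics and a toss rule onto its parent consumable ability.
; After the montage triggers post-commit, it adds a looping shield GameplayCue and starts the weapon's liquid-consume FX; ability end removes the cue.
; Consume end resets the liquid FX and reinitializes its Niagara system.
; Projectile toss is allowed only when enabled, the ability has triggered, and healing level is below max ticks.
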(event Ability|EventOnEndAbility
  (params (bWasCancelled :type Boolean))
  (GameplayCue|RemoveGameplayCueOnActor(Looping) (Ability|GetAvatarActorfromActorInfo) (Variables|Default|GetGC_ShieldsLoop) (Ability|GameplayCue|MakeGameplayCueParameters 0.0))
  (|CallParentFunction bWasCancelled))

(fn CheckCanTossPrj
  ;[0.4546 0.06953 0.6667] Check if max ticks were NOT reached in order to do the toss
  (return (Math|Boolean|ANDBoolean (Variables|Default|GetShouldTossPrj?) (Variables|Default|GetAbilityHasBeenTriggered) (< (Variables|Default|GetHealingLevel) (ScalableFloat|GetValueasInteger (Variables|Default|GetHF_MaxTicks))))))

(fn OnMontageTriggeredPostCommitSuccess
  (|CallParentFunction)
  (GameplayCue|AddGameplayCueOnActor(Looping) (Ability|GetAvatarActorfromActorInfo) (Variables|Default|GetGC_ShieldsLoop) (Ability|GameplayCue|MakeGameplayCueParameters 0.0))
  (bind _asb_consumable_small_half_shield_paprika (Utilities|Casting|CastToB_ConsumableSmall_HalfShield_Paprika (Ability|GetCurrentSourceWeapon)))
  (Class|BConsumableSmallHalfShieldAthena|ConsumeLiquidFX _asb_consumable_small_half_shield_paprika 4.0))

(fn PostConsumeEnd
  (|CallParentFunction)
  (bind _asb_consumable_small_half_shield_paprika (Utilities|Casting|CastToB_ConsumableSmall_HalfShield_Paprika (Ability|GetCurrentSourceWeapon)))
  (Utilities|IsValid _asb_consumable_small_half_shield_paprika
    (:Is%20Valid
      (Class|BConsumableSmallHalfShieldAthena|ResetLiquidFX _asb_consumable_small_half_shield_paprika false)
      (Niagara|ReinitializeSystem (Variables|/Game/Abilities/Player/Generic/UtilityItems/B_ConsumableSmall_HalfShield_Athena.B_ConsumableSmall_HalfShield_Athena_C|GetLiquid _asb_consumable_small_half_shield_paprika)))))
